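# Niagara system for the Caution Tape player emote, drawing a yellow tape and glow attached to a skeletal mesh.
# Caution_Tape and TapeGlow stretch ribbon beams between two sampled mesh points, with the glow tapering at both ends and animating its end offset.
# Caution_Tape001 places two sprites at another pair of sampled points.
# Glow and sprite alphas fade in by emitter age with waveform modulation; the LWE emitters list no renderers.
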NiagaraSystem /BRCosmetics/Animation/Game/MainPlayer/Emotes/CautionTape/VFX/NS_Emote_CautionTape
dependencies {
 usedDataInterfaces [&NiagaraDataInterfaceVectorCurve, &NiagaraDataInterfaceCurve, &NiagaraDataInterfaceSkeletalMesh]
 usedDynamicInputs []
 usedModules [
  &self:SystemSpawnScript.SystemScriptSource.SystemScriptGraph.NiagaraNodeAssignment_0.SetVariables_#1
  &SystemState
 ]
}
emitters {
 AirglowLWE {rendererClasses [], renderers []}
 AirglowLWE001 {rendererClasses [], renderers []}
 Caution_Tape {
  emitterUpdateScript [
   &EmitterState(Scalability Mode=::NewEnumerator1)
   &SampleSkeletalMesh(
    AdditionalYawPitchRoll=false
    Bone Sampling Mode:ENiagaraBoneSamplingMode=::NewEnumerator4
    Position Offset Coordinate Space:ENiagaraCoordinateSpace=::Local
    Sampled Position Offset:Vector3f=<10 0 -1.2256>
    Skeletal Mesh:NiagaraDataInterfaceSkeletalMesh=$System.Skeletal Mesh
   )
   &SampleSkeletalMesh(
    AdditionalYawPitchRoll=false
    Bone / Socket Index=1
    Bone Sampling Mode:ENiagaraBoneSamplingMode=::NewEnumerator4
    Position Offset Coordinate Space:ENiagaraCoordinateSpace=::Local
    Sampled Position Offset:Vector3f=<0 0 -1.2256>
    Skeletal Mesh:NiagaraDataInterfaceSkeletalMesh=$System.Skeletal Mesh
   ) {module=SampleSkeletalMesh001}
   /Niagara/Modules/Beams/BeamEmitterSetup(
    Absolute Beam End=true
    Beam End:NiagaraPosition=$Emitter.SampleSkeletalMesh001.SampledPosition
    Beam End Tangent:Vector3f=<0 0 0>
    Beam Start:NiagaraPosition=$Emitter.SampleSkeletalMesh.SampledPosition
    Beam Start Tangent:Vector3f=<0 0 0>
    Use Beam Tangents=true
   )
   &SpawnBurst_Instantaneous(Spawn Count=2, Use Spawn Probability=false, UseLoopCountLimit=false)
  ]
  particleSpawnScript [
   &InitializeParticle(
    Color=<rgba 0.6105 0.3916 0 1>
    Ribbon Width=1
    Ribbon Width Mode=::NewEnumerator1
    Sprite Size Mode=::NewEnumerator3
    UsePositionOffset=false
    Write Material Random=true
   )
   /Niagara/Modules/Beams/SpawnBeam()
   /Niagara/Modules/Beams/BeamWidth(Beam Width=11)
  ]
  particleUpdateScript [
   &ParticleState(Kill Particles When Lifetime Has Elapsed=false, Loop Particles Lifetime=true)
   /Niagara/Modules/Beams/UpdateBeam()
   /SetVariables() {isSetParameters=true, module=SetVariables_#2}
   &Color(Color=<rgba 6 3 0.021 1>)
  ]
  rendererClasses [&NiagaraRibbonRendererProperties]
  renderers [
   {rendererClass=&NiagaraRibbonRendererProperties}
   {rendererClass=&NiagaraRibbonRendererProperties, rendererIndex=1}
  ]
 }
 Caution_Tape001 {
  emitterUpdateScript [
   &EmitterState(Scalability Mode=::NewEnumerator1)
   &SampleSkeletalMesh(
    AdditionalYawPitchRoll=false
    Bone / Socket Index=2
    Bone Sampling Mode:ENiagaraBoneSamplingMode=::NewEnumerator4
    Position Offset Coordinate Space:ENiagaraCoordinateSpace=::Local
    Sampled Position Offset:Vector3f=<0 0 85>
    Skeletal Mesh:NiagaraDataInterfaceSkeletalMesh=$System.Skeletal Mesh
   )
   &SampleSkeletalMesh(
    AdditionalYawPitchRoll=false
    Bone / Socket Index=3
    Bone Sampling Mode:ENiagaraBoneSamplingMode=::NewEnumerator4
    Position Offset Coordinate Space:ENiagaraCoordinateSpace=::Local
    Sampled Position Offset:Vector3f=<0 0 85>
    Skeletal Mesh:NiagaraDataInterfaceSkeletalMesh=$System.Skeletal Mesh
   ) {module=SampleSkeletalMesh001}
   &SpawnBurst_Instantaneous(Spawn Count=2, Use Spawn Probability=false, UseLoopCountLimit=false)
  ]
  particleSpawnScript [&InitializeParticle(
    Color=<rgba 0.6105 0.3916 0 1>
    Ribbon Width=1
    Ribbon Width Mode=::NewEnumerator1
    Sprite Size=<40 20>
    Sprite Size Mode=::NewEnumerator4
    UsePositionOffset=false
    Write Material Random=true
   )]
  particleUpdateScript [
   &ParticleState(Kill Particles When Lifetime Has Elapsed=false, Loop Particles Lifetime=true)
   /SetVariables(Particles.Position=&LerpPosition(
     A=$Emitter.SampleSkeletalMesh.SampledPosition
     B=$Emitter.SampleSkeletalMesh001.SampledPosition
     Alpha=&ReturnNormalizedExecIndex()
    )) {isSetParameters=true}
   &Color(
    Color=<rgba 0.3906 0.2692 0.0016 1>
    Scale Alpha=&FloatFromCurve(
     FloatCurve=DataInterface() {propertyValues {Curve {keys [<at -0.6 0>,<at 0.2 0>,<at 2 1>]}}}
     CurveIndex=$Emitter.Age
     Scale Curve=&Waveform("Waveform [1]"=::NewEnumerator0, "[1] Frequency"=4, "[1] Amplitude Min/Max"=<0.6 0.8>)
    )
   )
  ]
  rendererClasses [&NiagaraSpriteRendererProperties]
  renderers [{rendererClass=&NiagaraSpriteRendererProperties}]
 }
 SmokeLWE {rendererClasses [], renderers []}
 SmokeLWE001 {rendererClasses [], renderers []}
 SparklesLWE {rendererClasses [], renderers []}
 SparklesLWE001 {rendererClasses [], renderers []}
 TapeGlow {
  emitterUpdateScript [
   &EmitterState(Scalability Mode=::NewEnumerator1)
   &SampleSkeletalMesh(
    AdditionalYawPitchRoll=false
    Bone Sampling Mode:ENiagaraBoneSamplingMode=::NewEnumerator4
    Position Offset Coordinate Space:ENiagaraCoordinateSpace=::Local
    Sampled Position Offset:Vector3f=<10 0 0>
    Skeletal Mesh:NiagaraDataInterfaceSkeletalMesh=$System.Skeletal Mesh
   )
   &SampleSkeletalMesh(
    AdditionalYawPitchRoll=false
    Bone / Socket Index=1
    Bone Sampling Mode:ENiagaraBoneSamplingMode=::NewEnumerator4
    Position Offset Coordinate Space:ENiagaraCoordinateSpace=::Local
    Sampled Position Offset:Vector3f=&VectorFromCurve(
     VectorCurve=DataInterface() {propertyValues {XCurve {keys [<at 0 0>,<at 1.5 0>,<at 2 -10>]}, YCurve {keys [<at 0 1>,<at 1 0>]}, ZCurve {keys [<at 0 0>,<at 0 0>]}}}
     CurveIndex=$Emitter.Age
    )
    Skeletal Mesh:NiagaraDataInterfaceSkeletalMesh=$System.Skeletal Mesh
   ) {module=SampleSkeletalMesh001}
   &BeamEmitterSetup(
    Absolute Beam End=true
    Beam End:NiagaraPosition=$StackContext.SampleSkeletalMesh001.SampledPosition
    Beam End Tangent:Vector3f=<0 0 0>
    Beam Start:NiagaraPosition=$StackContext.SampleSkeletalMesh.SampledPosition
    Beam Start Tangent:Vector3f=<0 0 0>
    Use Beam Tangents=true
   )
   &SpawnBurst_Instantaneous(Spawn Count=8, Use Spawn Probability=false, UseLoopCountLimit=false)
  ]
  particleSpawnScript [
   &InitializeParticle(
    Color=<rgba 0.6105 0.3916 0 1>
    Ribbon Width=1
    Ribbon Width Mode=::NewEnumerator1
    Sprite Size Mode=::NewEnumerator3
    Uniform Sprite Size=15
    UsePositionOffset=true
    Write Material Random=true
   )
   &SpawnBeam()
   &BeamWidth(Beam Width:NiagaraFloat=&FloatFromCurve(
     FloatCurve=DataInterface() {propertyValues {Curve {keys [
         <at 0 0.5 RCIM_Cubic>
         {arriveTangent=0.8125, time=0.1, value=1}
         {arriveTangent=-0.5556, interpMode=RCIM_Cubic, leaveTangent=-0.5556, time=0.9, value=1}
         <at 1 0.5 RCIM_Cubic>
        ]}}}
     CurveIndex=&ReturnNormalizedExecIndex()
     Scale Curve=18
    ))
   /SetVariables(Particles.VisibilityTag=&SpawnGroup()) {isSetParameters=true}
  ]
  particleUpdateScript [
   &ParticleState(Kill Particles When Lifetime Has Elapsed=false, Loop Particles Lifetime=true)
   &DynamicMaterialParameters(
    Index 0 Param 1=-1
    Module.Index0Float0Write=true
    Module.Index0Float1Write=true
    Module.Index0Float2Write=true
    Module.Index0Float3Write=true
    Module.Index1Float0Write=true
    Module.Index1Float1Write=true
    Module.Index1Float2Write=true
    Module.Index1Float3Write=true
    Module.Index2Float0Write=true
    Module.Index2Float1Write=true
    Module.Index2Float2Write=true
    Module.Index2Float3Write=true
    Module.Index3Float0Write=true
    Module.Index3Float1Write=true
    Module.Index3Float2Write=true
    Module.Index3Float3Write=true
    Write Parameter Index 1=::NewEnumerator1
    Write Parameter Index 2=::NewEnumerator1
    Write Parameter Index 3=::NewEnumerator1
   )
   &UpdateBeam()
   /SetVariables() {isSetParameters=true, module=SetVariables_#2}
   &Color(
    Color=<rgba 1.5 1 0.021 0.6>
    Scale Alpha=&FloatFromCurve(
     FloatCurve=DataInterface() {propertyValues {Curve {keys [<at 0 0>,<at 1.1 0>,<at 2 1>]}}}
     CurveIndex=$Emitter.Age
     Scale Curve=&Waveform("Waveform [1]"=::NewEnumerator0, "[1] Frequency"=1, "[1] Amplitude Min/Max"=<0.7 0.8>)
    )
   )
  ]
  rendererClasses [&NiagaraRibbonRendererProperties]
  renderers [{rendererClass=&NiagaraRibbonRendererProperties}]
 }
}
summary {
 systemName=NS_Emote_CautionTape
 emitters [
  {emitterName=SparklesLWE, rendererClasses []}
  {emitterName=AirglowLWE, rendererClasses []}
  {emitterName=SmokeLWE, rendererClasses []}
  {emitterName=SparklesLWE001, rendererClasses []}
  {emitterName=AirglowLWE001, rendererClasses []}
  {emitterName=SmokeLWE001, rendererClasses []}
  {emitterName=Caution_Tape, rendererClasses [&NiagaraRibbonRendererProperties]}
  {emitterName=TapeGlow, rendererClasses [&NiagaraRibbonRendererProperties]}
  {emitterName=Caution_Tape001, rendererClasses [&NiagaraSpriteRendererProperties]}
 ]
 userVariables [User.NormalizedNotifyProgress {type=NiagaraFloat, defaultValue=0, description=""}]
}
